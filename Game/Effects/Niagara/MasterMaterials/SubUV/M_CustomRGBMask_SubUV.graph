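# Translucent, unlit, two-sided master material for Niagara sprite and mesh particles, drawing shapes from a channel-packed Debris texture parameter that defaults to outlined ember pieces.
# A static channel mask (green by default) picks the color shape, which is multiplied by particle RGB and fed through MF_FX_EmissiveAdjustments.
# A second static mask (blue by default) picks the alpha shape, multiplied by particle alpha.
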Material /Game/Effects/Niagara/MasterMaterials/SubUV/M_CustomRGBMask_SubUV
graph {
 properties {
  bAutomaticallySetUsageInEditor=false
  bUsedWithNiagaraMeshParticles=true
  bUsedWithNiagaraSprites=true
  blendMode=BLEND_Translucent
  shadingModel=MSM_Unlit
  twoSided=true
  usedShadingModels=Unlit
 }
 nodes [
  n1 = TextureSampleParameter2D() {parameterName=Debris, texture=/Game/Effects/Fort_Effects/Textures/T_OutlinedEmbers_Pieces}
  MaterialFunctionCall(Emissive=Multiply(A=ParticleColor().RGB, B=StaticComponentMaskParameter(None=@n1.RGB) {defaultG=true, parameterName=DebrisMask})) {materialFunction=&MF_FX_EmissiveAdjustments}
  Multiply(
   A=StaticComponentMaskParameter(None=TextureSampleParameter2D() {parameterName=Debris, texture=&T_OutlinedEmbers_Pieces}.RGB) {defaultB=true, parameterName=AlphaMask}
   B=ParticleColor().A
  )
 ]
}
