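# Unlit, two-sided translucent master material for Niagara sprite and mesh particles, defaulting to a 2x2 spikes debris texture.
# Particle color, optionally tinted by the texture and desaturated, feeds a shared emissive-adjustment function.
# Opacity combines a selectable texture channel with power control, particle alpha and optional Perlin-noise HeightLerp dissolve driven by the Dissolve dynamic parameter.
# Depth fade and optional view-angle fade soften the result.
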Material /Game/Effects/Niagara/MasterMaterials/SubUV/M_Spikes2x2
graph {
 properties {
  bAutomaticallySetUsageInEditor=false
  bUsedWithNiagaraMeshParticles=true
  bUsedWithNiagaraSprites=true
  blendMode=BLEND_Translucent
  shadingModel=MSM_Unlit
  twoSided=true
  usedShadingModels=Unlit
 }
 nodes [
  n1 = ParticleColor()
  n2 = LinearInterpolate(
   B=TextureSampleParameter2D() {parameterName=Debris, texture=/Game/Effects/Niagara/MasterMaterials/Spine/Texture/T_Spikes2x2}.RGB
   Alpha=ScalarParameter() {parameterName=UseTextureOnEmissive}
  ) {constA=1}
  MaterialFunctionCall(Emissive=Desaturation(None=Multiply(A=@n1.RGB, B=@n2), Fraction=ScalarParameter() {parameterName=Des})) {materialFunction=&MF_FX_EmissiveAdjustments}
  n3 = TextureSampleParameter2D() {parameterName=Debris, texture=&T_Spikes2x2}
  n4 = Multiply(
   A=Power(
    Base=StaticComponentMaskParameter(None=@n3.RGBA) {defaultR=true, parameterName=RGBMask}
    Exp=ScalarParameter() {defaultValue=1, parameterName=AlphaPower}
   )
   B=ParticleColor().A
  )
  n5 = DynamicParameter() {defaultValue=<rgba 1 1 1 0>, paramNames [Param1,Param2,Param3,Dissolve]}
  n6 = VectorParameter() {parameterName=DissUV, defaultValue=<rgba 1 1 0 0>}
  n7 = TextureSampleParameter2D(UVs=Multiply(A=TextureCoordinate(), B=ComponentMask(None=@n6.RGB) {g=true, r=true})) {
   parameterName=Tex_Dissolve
   samplerType=SAMPLERTYPE_LinearGrayscale
   texture=&T_SimplePerlin_Greyscale
  }
  n8 = LinearInterpolate(
   B=MaterialFunctionCall(
    Transition Phase=LinearInterpolate(Alpha=@n5.Dissolve) {constA=0.3, constB=0.8}
    Height Texture=@n7.R
    Contrast=ScalarParameter() {defaultValue=3, parameterName="Contrast (S)"}
   ) {materialFunction=&HeightLerp}.Results
   Alpha=ScalarParameter() {parameterName=UseDissolve}
  ) {constA=1}
  n9 = Constant() {r=1}
  Multiply(
   A=DepthFade(Opacity=Multiply(A=@n4, B=@n8), FadeDistance=ScalarParameter() {defaultValue=10, parameterName=FadeDistance})
   B=LinearInterpolate(
    B=MaterialFunctionCall("FadeSensitivity (optional)"=@n9) {materialFunction=&FadeBasedOnViewAngle}.Result
    Alpha=ScalarParameter() {parameterName=UseAngleFade}
   ) {constA=1}
  )
 ]
}
